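# Mono MetaSound Source for the looping energy layer of the DragonCart orb shock sequence.
# On play, a single Wave Player triggers the exposed Sound wave input, defaulting to the EnergyLP_A wave, and feeds the mono output directly.
# The source is flagged looping, uses the shock-sequence energy beam attenuation and routes to the Fort_Ambient_Athena sound class.
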MetaSoundSource /DragonCartNarrative/Assets/Audio/OrbShockSequence/MSS_DragonCart_ShockSeq_EnergyLP_A
inputs [
 Sound {type=WaveAsset}
 UE.Source.OnPlay {type=Trigger}
]
interfaces [UE.OutputFormat.Mono:1.0, UE.Source:1.1]
outputs ["UE.OutputFormat.Mono.Audio:0" {type=Audio}]
pages [Default {
  input_defaults {Sound=/DragonCartNarrative/Assets/Audio/OrbShockSequence/Wav/DragonCart_ShockSeq_EnergyLP_A, UE.Source.OnPlay=null}
  nodes [WavePlayer_0 = UE.Wave Player:Mono(Wave Asset=^Sound, Play=^UE.Source.OnPlay) {also {defaults {Loop="<MetaSoundBuilderResult.SUCCEEDED: 0>"}, version [1,1]}}]
  links [{dst="@UE.OutputFormat.Mono.Audio:0", src=WavePlayer_0.Out Mono}]
 }]
properties {
 assetImportData=&self:AssetImportData
 attenuationSettings=&/ATT_DragonCart_ShockSeq_EnergyBeam
 bLooping=true
 duration=10000
 maxDistance=10000
 sampleRate=48000
 soundClassObject=&Fort_Ambient_Athena
 volume=2
}
