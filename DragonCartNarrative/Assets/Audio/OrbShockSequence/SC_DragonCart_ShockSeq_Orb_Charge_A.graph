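# Sound cue for the orb charge sound in the DragonCart narrative's orb shock sequence.
# A single wave player feeds the output directly, with no randomization or mixing, at slightly reduced volume.
# It uses the orb shock sequence attenuation asset and the Fort_Events sound class.
# It also carries owner-limited concurrency override values with a small voice cap and a retrigger time, while its concurrency set holds only a null entry.
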SoundCue /DragonCartNarrative/Assets/Audio/OrbShockSequence/SC_DragonCart_ShockSeq_Orb_Charge_A
root=WavePlayer_1
nodes [{id=WavePlayer_1, properties {soundWaveAssetPtr=/DragonCartNarrative/Assets/Audio/OrbShockSequence/Wav/DragonCart_ShockSeq_Orb_Charge_A}}]
properties {
 attenuationSettings=&/ATT_DragonCart_Orb_ShockSeq
 maxDistance=84000
 soundClassObject=&Fort_Events
 volumeMultiplier=0.825
 concurrencyOverrides {bLimitToOwner=true, maxCount=2, retriggerTime=0.25}
 concurrencySet [null]
}
