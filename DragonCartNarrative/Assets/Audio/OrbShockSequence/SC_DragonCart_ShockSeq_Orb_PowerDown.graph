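# Sound cue for the orb power-down in the DragonCart shock sequence, part of the DragonCart narrative audio.
# A single wave player feeds the output directly, with no randomization or modulation.
# It uses the shock-sequence orb attenuation with a very large audible range and routes through the Fort_Events sound class.
# Concurrency overrides limit simultaneous instances per owner and set a retrigger time.
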SoundCue /DragonCartNarrative/Assets/Audio/OrbShockSequence/SC_DragonCart_ShockSeq_Orb_PowerDown
root=WavePlayer_0
nodes [{id=WavePlayer_0, properties {soundWaveAssetPtr=/DragonCartNarrative/Assets/Audio/OrbShockSequence/Wav/DragonCart_ShockSeq_Orb_PowerDown}}]
properties {
 attenuationSettings=&/ATT_DragonCart_Orb_ShockSeq
 maxDistance=84000
 soundClassObject=&Fort_Events
 concurrencyOverrides {bLimitToOwner=true, maxCount=2, retriggerTime=0.25}
 concurrencySet [null]
}
