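# Looping sound cue for the energy spline shard audio in the DragonCart shock sequence.
# A single looping wave player feeds the output directly, shaped by the shard-spline attenuation asset and routed through the Fort_Events sound class.
# Powered-spline concurrency and seek-restart virtualization manage how instances play and resume across its long audible range.
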SoundCue /DragonCartNarrative/Assets/Audio/OrbShockSequence/SC_DragonCart_Shock_Seq_Shard_EnergySplines_LP
root=WavePlayer_1
nodes [{
  id=WavePlayer_1
  properties {bLooping=true, soundWaveAssetPtr=/DragonCartNarrative/Assets/Audio/OrbShockSequence/Wav/DragonCart_ShockSeq_Shard_EnergySpline_LP}
 }]
properties {
 attenuationSettings=&/ATT_DragonCart_ShardSplines
 maxDistance=80000
 soundClassObject=&Fort_Events
 virtualizationMode=SeekRestart
 volumeMultiplier=1
 concurrencyOverrides {bLimitToOwner=true, maxCount=2, retriggerTime=0.25}
 concurrencySet [&/CONC_DragonCart_PoweredSplines]
}
